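; Pawn Blueprint for the Caretaker, a monster NPC tied to the Sideways world.
; On the server, its Sideways state toggles invulnerability, faction and a cooldown effect.
; Clients swap materials by the local or spectated player's Sideways state, with FX and audio layered on.
; It also handles timer-faked footstep damage, notify-driven weapon capsule resizing, beam sweeps, hit-location reactions, safe zone and encounter tracking, and possession-based debug controls.
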(event Custom|Debug_Caretaker_ForceMeleeOnly
  (bind _aipawnabilitysystemcomponent (Variables|GameplayAbilities|GetAIPawnAbilitySystemComponent))
  ;[1 0.0304 0] Debug Functionality For Caretaker
  (GameplayEffects|RemoveActiveEffectswithSourceTags _aipawnabilitysystemcomponent "(GameplayTags=((TagName=\"NPC.Ability.Debug.Solo.RemoveAll\")))")
  (GameplayEffects|ApplyGameplayEffectToSelf :self _aipawnabilitysystemcomponent :GameplayEffectClass "/Caretaker/Abilities/Debug/GE_NPC_Irwin_Monster_Caretaker_Debug_AbilitySolo_Attack_MeleeAoE.GE_NPC_Irwin_Monster_Caretaker_Debug_AbilitySolo_Attack_MeleeAoE_C" :Level 0.0))

(event Custom|RegisterToLocalClientSidewaysComponent
  ;[1 1 1] Track Local Pawn Enter / Exit sideways. Register to pawn on local client
  (bind _array_element (Utilities|Array|ForEachLoopwithBreak (Fortnite|GetLocalFortPlayerControllers true))
    (:LoopBody
      (Utilities|IsValid (Pawn|GetControlledPawn _array_element)
        (:Is%20Valid
          (bind _returnvalue (Sideways|GetSidewaysStateComponent (Pawn|GetControlledPawn _array_element)))
          (Utilities|IsValid _returnvalue
            (:Is%20Valid
              (Variables|Default|SetisMaterialHandledFromPlayer true)
              (CallFunction|SetCaretakerMaterials (Utilities|Enum|Equal(Enum) (Class|SidewaysStateComponent|GetSidewaysState _returnvalue) "InSidewaysWorld")) ; Match whatever state the player is already in
              (EventDispatchers|AssignOnSidewaysStateChangedDelegate _returnvalue (delegate Custom|OnSidewaysStateChangedDelegate_LocalPlayer)))
            (:Is%20Not%20Valid
              _))
          ;;[1 1 1] Handle spectators visibility on caretaker
          (bind _asfort_player_state_athena (Utilities|Casting|CastToFortPlayerStateAthena (Variables|/Script/Engine.Pawn|GetPlayerState (Pawn|GetControlledPawn _array_element))))
          (EventDispatchers|AssignOnSpectatingTargetChangedEvent _asfort_player_state_athena (delegate Custom|OnSpectatingTargetChangedEvent_Event_0)))
        (:Is%20Not%20Valid
          _)))))

(event EventBeginPlay
  ;[1 1 1] Begin Play
  (|CallParentFunction)
  (AI|SetMinSignificanceforAnimationCurves 0.0)
  (bind _returnvalue (Utilities|Time|SetTimerbyEvent (EventDispatchers|CreateEvent :!event Custom|UnhideCaretakerAfterDelay) 2.0))
  (Variables|Default|SetCaretakerUnhideTimer _returnvalue)
  (Networking|SwitchHasAuthority
    (:Authority
      (CallFunction|ExitedSideways)
      (CallFunction|BeginCaretakerSafeZoneTracking)
      (CallFunction|WeaponCapsuleSetStartingValues)
      (CallFunction|FootstepDamageHandling)
      (bind _returnvalue_2 (Sideways|GetSidewaysStateComponent self))
      (Utilities|IsValid _returnvalue_2
        (:Is%20Valid
          _)
        (:Is%20Not%20Valid
          (Utilities|FlowControl|DelayUntilNextTick)))
      (bind _output_get (Variables|Components|SetSidewaysComponent _returnvalue_2))
      (EventDispatchers|AssignOnSidewaysStateChangedDelegate _output_get (EventDispatchers|CreateEvent :!event Custom|OnSidewaysStateChangedDelegate_Event_0)))
    (:Remote
      (CallFunction|ExitedSidewaysClient)
      (CallFunction|SetCaretakerMaterials false)
      (CallFunction|RegistertoLocalClientSidewaysComponent)
      (CallFunction|WeaponCapsuleSetStartingValues)
      ;[1 1 1] Hide Caretaker On Spawn, force unhide after short duration
      (Rendering|SetActorHiddenInGame true))))

(event Custom|OnSpectatingTargetChangedEvent_Event_0
  (params (SpectatingTarget :type Fort%20Player%20State%20Zone%20Object%20Reference))
  (bind _returnvalue (PlayerState|GetCurrentPawn SpectatingTarget))
  (bind _previousspectatepawnlocalplayer (Variables|Default|GetpreviousSpectatePawnLocalPlayer))
  ;[1 1 1] Track Local Pawn Enter / Exit sideways. Register to pawn on local client
  ;;[1 1 1] Handle spectators visibility on caretaker
  (Utilities|FlowControl|Sequence
    (:then_0
      (Variables|Default|SetisMaterialHandledFromPlayer true)
      (Utilities|IsValid _previousspectatepawnlocalplayer
        (:Is%20Valid
          (bind _returnvalue_2 (Sideways|GetSidewaysStateComponent _previousspectatepawnlocalplayer))
          (Utilities|IsValid _returnvalue_2
            (:Is%20Valid
              (EventDispatchers|UnbindEventfromOnSidewaysStateChangedDelegate _returnvalue_2 (EventDispatchers|CreateEvent :!event Custom|OnSidewaysStateChangedDelegate_LocalPlayer))
              (CallFunction|SetCaretakerMaterials (Utilities|Enum|Equal(Enum) (Class|SidewaysStateComponent|GetSidewaysState _returnvalue_2) "InSidewaysWorld")) ; Match whatever state the player is already in
            )))))
    (:then_1
      (Utilities|IsValid _returnvalue
        (:Is%20Valid
          (Variables|Default|SetpreviousSpectatePawnLocalPlayer _returnvalue)
          (bind _returnvalue_3 (Sideways|GetSidewaysStateComponent _returnvalue))
          (Utilities|IsValid _returnvalue_3
            (:Is%20Valid
              (EventDispatchers|AssignOnSidewaysStateChangedDelegate _returnvalue_3 (EventDispatchers|CreateEvent :!event Custom|OnSidewaysStateChangedDelegate_LocalPlayer)))))))))

(event Custom|Debug_Caretaker_ForceSnapOnly
  (bind _aipawnabilitysystemcomponent (Variables|GameplayAbilities|GetAIPawnAbilitySystemComponent))
  ;[1 0.0304 0] Debug Functionality For Caretaker
  (GameplayEffects|RemoveActiveEffectswithSourceTags _aipawnabilitysystemcomponent "(GameplayTags=((TagName=\"NPC.Ability.Debug.Solo.RemoveAll\")))")
  (GameplayEffects|ApplyGameplayEffectToSelf :self _aipawnabilitysystemcomponent :GameplayEffectClass "/Caretaker/Abilities/Debug/GE_NPC_Irwin_Monster_Caretaker_Debug_AbilitySolo_Attack_Snap.GE_NPC_Irwin_Monster_Caretaker_Debug_AbilitySolo_Attack_Snap_C" :Level 0.0))

(event Custom|Debug_Caretaker_ForceSwipeOnly
  (bind _aipawnabilitysystemcomponent (Variables|GameplayAbilities|GetAIPawnAbilitySystemComponent))
  ;[1 0.0304 0] Debug Functionality For Caretaker
  (GameplayEffects|RemoveActiveEffectswithSourceTags _aipawnabilitysystemcomponent "(GameplayTags=((TagName=\"NPC.Ability.Debug.Solo.RemoveAll\")))")
  (GameplayEffects|ApplyGameplayEffectToSelf :self _aipawnabilitysystemcomponent :GameplayEffectClass "/Caretaker/Abilities/Debug/GE_NPC_Irwin_Monster_Caretaker_Debug_AbilitySolo_Attack_Swipe.GE_NPC_Irwin_Monster_Caretaker_Debug_AbilitySolo_Attack_Swipe_C" :Level 0.0))

(event Custom|OnSidewaysStateChangedDelegate_LocalPlayer
  (params (SidewaysStateComponent :type Sideways%20State%20Component%20Object%20Reference) (SidewaysPlayspace :type Sideways%20Playspace%20Object%20Reference) (OldState :type ESidewaysState%20Enum :default "InNormalWorld") (NewState :type ESidewaysState%20Enum :default "InNormalWorld"))
  (bind _returnvalue (Utilities|Enum|Equal(Enum) NewState "InSidewaysWorld"))
  ;[1 1 1] Track Local Pawn Enter / Exit sideways. Register to pawn on local client
  ;;[1 1 1] Sideways state changed tracking on players
  (if (Utilities|Enum|NotEqual(Enum) OldState NewState)
    (CallFunction|SetCaretakerMaterials _returnvalue)
    (Variables|Default|SetIsLocalPlayerInSideways _returnvalue)))

(event Custom|UnhideCaretakerAfterDelay
  ;[1 1 1] Hide Caretaker On Spawn, force unhide after short duration
  (Networking|SwitchHasAuthority
    (:Authority
      (Variables|Default|SetisCaretakerHidden false))
    (:Remote
      (Rendering|SetActorHiddenInGame false))))

(event Custom|Debug_Caretaker_ForceSidewaysOnly
  (bind _aipawnabilitysystemcomponent (Variables|GameplayAbilities|GetAIPawnAbilitySystemComponent))
  ;[1 0.0304 0] Debug Functionality For Caretaker
  (GameplayEffects|RemoveActiveEffectswithSourceTags _aipawnabilitysystemcomponent "(GameplayTags=((TagName=\"NPC.Ability.Debug.Solo.RemoveAll\")))")
  (GameplayEffects|ApplyGameplayEffectToSelf :self _aipawnabilitysystemcomponent :GameplayEffectClass "/Caretaker/Abilities/Debug/GE_NPC_Irwin_Monster_Caretaker_Debug_AbilitySolo_CastSideways.GE_NPC_Irwin_Monster_Caretaker_Debug_AbilitySolo_CastSideways_C" :Level 0.0))

(event EventEndPlay
  (params (EndPlayReason :type EEndPlayReason%20Enum :default "Destroyed"))
  ;[1 1 1] End Play
  (|CallParentFunction EndPlayReason)
  (CallFunction|StopFootStepDamageHandling)
  (Utilities|Time|ClearandInvalidateTimerbyHandle (Variables|SafeZoneTracking|GetCaretakerSafeZoneTrackingTimerHandle))
  (Utilities|Time|ClearandInvalidateTimerbyHandle (Variables|Default|GetCaretakerUnhideTimer)))

(event Custom|Debug_Caretaker_RemoveCooldownsFromAbilities
  ;[1 0.0304 0] Debug Functionality For Caretaker
  (GameplayEffects|ApplyGameplayEffectToSelf :self (Variables|GameplayAbilities|GetAIPawnAbilitySystemComponent) :GameplayEffectClass "/Caretaker/Abilities/Debug/GE_NPC_Irwin_Monster_Caretaker_DebugReduceCooldowns.GE_NPC_Irwin_Monster_Caretaker_DebugReduceCooldowns_C" :Level 0.0))

(event EventPossessed
  (params (NewController :type Controller%20Object%20Reference))
  ;[1 1 1] Use event possessed to add some debug functionality (only if via player controller)
  (|CallParentFunction NewController)
  (bind _asfort_player_controller (Utilities|Casting|CastToFortPlayerController NewController))
  (CallFunction|WatchforKeyEventsonLocalController _asfort_player_controller)
  (Networking|ForceNetUpdate))

(event Custom|SetCaretakerMaterials
  (params (isSideways :type Boolean))
  (bind _mesh (Variables|Character|GetMesh))
  ;[1 1 1] Track Local Pawn Enter / Exit sideways. Register to pawn on local client
  ;;[1 1 1] Handle Material Changes
  (Networking|SwitchHasAuthority
    (:Remote
      (if (Variables|Health|GetbIsDying)
        _
        (elif (Variables|Default|GetdebugSidewaysOverrideActive)
          _
          (elif isSideways
            (bind (_gamestate _playlist _playlistcontexttags) (Class|FortAsyncActionCurrentPlaylistReady|PlaylistDataReadyAsync)
              (:OnReady
                (if (GameplayTags|HasTag _playlistcontexttags "(TagName=\"Athena.Playlist.Mash\")" true)
                  (Rendering|Material|SetMaterial _mesh 0 "/Caretaker/Materials/M_Caretaker_Body_HordeRush.M_Caretaker_Body_HordeRush")
                  (Rendering|Material|SetMaterial _mesh 1 "/Caretaker/Materials/M_Caretaker_Limbs_HordeRush.M_Caretaker_Limbs_HordeRush")
                  (else
                    (Rendering|Material|SetMaterial _mesh 0 "/Caretaker/Materials/M_Caretaker_Body.M_Caretaker_Body")
                    (Rendering|Material|SetMaterial _mesh 1 "/Caretaker/Materials/M_Caretaker_Limbs.M_Caretaker_Limbs")
                    (Rendering|Material|SetMaterial _mesh 2 "/Game/Characters/Player/Male/Medium/Bodies/M_MED_HighTower_Tomato_Casual/Materials/M_VertexCrunch.M_VertexCrunch")
                    (Rendering|Material|SetMaterial _mesh 3 "/Game/Characters/Player/Male/Medium/Bodies/M_MED_HighTower_Tomato_Casual/Materials/M_VertexCrunch.M_VertexCrunch")))))
            (else
              (Rendering|Material|SetMaterial _mesh 0 "/Caretaker/Materials/M_Caretaker_Body_ET.M_Caretaker_Body_ET")
              (Rendering|Material|SetMaterial _mesh 1 "/Caretaker/Materials/M_Caretaker_Limbs_ET.M_Caretaker_Limbs_ET")
              (Rendering|Material|SetMaterial _mesh 3 "/Caretaker/Materials/M_CaretakerTransparency_Body.M_CaretakerTransparency_Body")
              (Rendering|Material|SetMaterial _mesh 2 "/Caretaker/Materials/M_CaretakerTransparency_Limbs.M_CaretakerTransparency_Limbs"))))))))

(event Custom|Debug_Caretaker_ResetAllDebug
  (bind _aipawnabilitysystemcomponent (Variables|GameplayAbilities|GetAIPawnAbilitySystemComponent))
  ;[1 0.0304 0] Debug Functionality For Caretaker
  (GameplayEffects|RemoveActiveEffectswithSourceTags _aipawnabilitysystemcomponent "(GameplayTags=((TagName=\"NPC.Ability.Debug.Solo.RemoveAll\")))")
  (GameplayEffects|RemoveActiveEffectswithSourceTags _aipawnabilitysystemcomponent "(GameplayTags=((TagName=\"NPC.Ability.Debug.ReduceCooldowns\")))")
  (GameplayEffects|RemoveActiveEffectswithSourceTags _aipawnabilitysystemcomponent "(GameplayTags=((TagName=\"NPC.Ability.Debug.Caretaker.IsInSideways\")))")
  (Variables|Default|SetdebugSidewaysOverrideActive false)
  (CallFunction|DebugToggleSidewaysMode (GameplayTags|HasMatchingGameplayTag "(TagName=\"Granted.Athena.Sideways.InsideSideways\")"))
  (CallFunction|DebugCloseMaterialParamsLoopMulticast))

(event Custom|OnSidewaysStateChangedDelegate_Event_0
  (params (SidewaysStateComponent :type Sideways%20State%20Component%20Object%20Reference) (SidewaysPlayspace :type Sideways%20Playspace%20Object%20Reference) (OldState :type ESidewaysState%20Enum :default "InNormalWorld") (NewState :type ESidewaysState%20Enum :default "InNormalWorld"))
  ;[0.6123 0 1] Sideways Handling
  (if (Utilities|Enum|NotEqual(Enum) OldState NewState)
    (switch Utilities|FlowControl|Switch|SwitchonESidewaysState NewState
      (:InNormalWorld
        (CallFunction|ExitedSideways)
        (Networking|ForceNetUpdate))
      (:InSidewaysWorld
        (CallFunction|EnteredSideways)
        (Networking|ForceNetUpdate)))))

(event Custom|Debug_Caretaker_ToggleSideways
  (bind _aipawnabilitysystemcomponent (Variables|GameplayAbilities|GetAIPawnAbilitySystemComponent))
  ;[1 0.0304 0] Debug Functionality For Caretaker
  (Variables|Default|SetdebugSidewaysOverrideActive true)
  (Utilities|FlowControl|FlipFlop
    (:A
      (GameplayEffects|ApplyGameplayEffectToSelf :self _aipawnabilitysystemcomponent :GameplayEffectClass "/Caretaker/Abilities/Debug/GE_NPC_Irwin_Monster_Caretaker_Debug_Sideways_InSideways.GE_NPC_Irwin_Monster_Caretaker_Debug_Sideways_InSideways_C" :Level 0.0)
      (CallFunction|DebugToggleSidewaysMode true))
    (:B
      (GameplayEffects|RemoveActiveEffectswithSourceTags _aipawnabilitysystemcomponent "(GameplayTags=((TagName=\"NPC.Ability.Debug.Caretaker.IsInSideways\")))")
      (CallFunction|DebugToggleSidewaysMode false))))

(event Custom|EnteredSideways
  (bind _self self)
  ;[0.6123 0 1] Sideways Handling
  ;;[1 0 0.02539] THIS SECTION IS SERVER ONLY
  (if (not (Variables|Default|GetdebugSidewaysOverrideActive))
    (Networking|SwitchHasAuthority
      (:Authority
        (Variables|Default|SetisInSideways true)
        (Variables|FortPawn|SetbIsInvulnerable false)
        (Affiliation|AddtoFaction _self "(TagName=\"Athena.Faction.Monster.Caretaker.Sideways\")")
        (Affiliation|RemovefromFaction _self "(TagName=\"Athena.Faction.Monster.Caretaker.Normal\")")
        (GameplayEffects|ApplyGameplayEffectToSelf :self (Variables|GameplayAbilities|GetAIPawnAbilitySystemComponent) :GameplayEffectClass "/Caretaker/Abilities/GE_NPC_Irwin_Monster_Caretaker_SidewaysGeneral_CooldownOverride.GE_NPC_Irwin_Monster_Caretaker_SidewaysGeneral_CooldownOverride_C" :Level 0.0)))))

(event Custom|Debug_ToggleSidewaysMode
  (params (enteringSideways :type Boolean))
  (bind _self self)
  (bind _self_2 self)
  ;[0.6123 0 1] Sideways Handling
  ;;[1 0 0.02539] THIS SECTION IS SERVER ONLY
  (if enteringSideways
    (Networking|SwitchHasAuthority
      (:Authority
        (Variables|Default|SetisInSideways true)
        (Variables|FortPawn|SetbIsInvulnerable false)
        (Affiliation|AddtoFaction _self_2 "(TagName=\"Athena.Faction.Monster.Caretaker.Sideways\")")
        (Affiliation|RemovefromFaction _self_2 "(TagName=\"Athena.Faction.Monster.Caretaker.Normal\")")
        (GameplayEffects|ApplyGameplayEffectToSelf :self (Variables|GameplayAbilities|GetAIPawnAbilitySystemComponent) :GameplayEffectClass "/Caretaker/Abilities/GE_NPC_Irwin_Monster_Caretaker_SidewaysGeneral_CooldownOverride.GE_NPC_Irwin_Monster_Caretaker_SidewaysGeneral_CooldownOverride_C" :Level 0.0)))
    (else
      (Networking|SwitchHasAuthority
        (:Authority
          (Variables|Default|SetisInSideways false)
          (Variables|FortPawn|SetbIsInvulnerable true)
          (Affiliation|RemovefromFaction _self "(TagName=\"Athena.Faction.Monster.Caretaker.Sideways\")")
          (Affiliation|AddtoFaction _self "(TagName=\"Athena.Faction.Monster.Caretaker.Normal\")")
          (GameplayEffects|RemoveActiveGameplayEffectbySourceEffect (Variables|GameplayAbilities|GetAIPawnAbilitySystemComponent) "/Caretaker/Abilities/GE_NPC_Irwin_Monster_Caretaker_SidewaysGeneral_CooldownOverride.GE_NPC_Irwin_Monster_Caretaker_SidewaysGeneral_CooldownOverride_C" 0 -1))))))

(event Custom|ExitedSideways
  (bind _self self)
  ;[0.6123 0 1] Sideways Handling
  ;;[1 0 0.02539] THIS SECTION IS SERVER ONLY
  (if (not (Variables|Default|GetdebugSidewaysOverrideActive))
    (Networking|SwitchHasAuthority
      (:Authority
        (Variables|Default|SetisInSideways false)
        (Variables|FortPawn|SetbIsInvulnerable true)
        (Affiliation|RemovefromFaction _self "(TagName=\"Athena.Faction.Monster.Caretaker.Sideways\")")
        (Affiliation|AddtoFaction _self "(TagName=\"Athena.Faction.Monster.Caretaker.Normal\")")
        (GameplayEffects|RemoveActiveGameplayEffectbySourceEffect (Variables|GameplayAbilities|GetAIPawnAbilitySystemComponent) "/Caretaker/Abilities/GE_NPC_Irwin_Monster_Caretaker_SidewaysGeneral_CooldownOverride.GE_NPC_Irwin_Monster_Caretaker_SidewaysGeneral_CooldownOverride_C" 0 -1)))))

(event Custom|Debug_Caretaker_ToggleMaterial
  ;[1 0.0304 0] Debug Functionality For Caretaker
  (CallFunction|DebugCaretakerToggleMaterialMulticast))

(event Custom|Debug_Caretaker_ToggleMaterial_Multicast
  (bind _fx_idle (Variables|Components|GetFX_Idle))
  (bind _mesh (Variables|Character|GetMesh))
  ;[1 0.0304 0] Debug Functionality For Caretaker
  (Networking|SwitchHasAuthority
    (:Remote
      (Utilities|FlowControl|FlipFlop
        (:A
          (Rendering|Material|SetMaterial _mesh 0 "/Caretaker/Materials/M_Caretaker_Body.M_Caretaker_Body")
          (Rendering|Material|SetMaterial _mesh 1 "/Caretaker/Materials/M_Caretaker_Limbs.M_Caretaker_Limbs")
          (Rendering|Material|SetMaterial _mesh 2 "/Game/Characters/Player/Male/Medium/Bodies/M_MED_HighTower_Tomato_Casual/Materials/M_VertexCrunch.M_VertexCrunch")
          (Rendering|Material|SetMaterial _mesh 3 "/Game/Characters/Player/Male/Medium/Bodies/M_MED_HighTower_Tomato_Casual/Materials/M_VertexCrunch.M_VertexCrunch")
          (Components|Activation|Deactivate _fx_idle)
          (CallFunction|DebugSetMaterialParamsLoop))
        (:B
          (Rendering|Material|SetMaterial _mesh 0 "/Caretaker/Materials/M_Caretaker_Body_ET.M_Caretaker_Body_ET")
          (Rendering|Material|SetMaterial _mesh 1 "/Caretaker/Materials/M_Caretaker_Limbs_ET.M_Caretaker_Limbs_ET")
          (Rendering|Material|SetMaterial _mesh 2 "/Caretaker/Materials/M_CaretakerTransparency_Body.M_CaretakerTransparency_Body")
          (Rendering|Material|SetMaterial _mesh 3 "/Caretaker/Materials/M_CaretakerTransparency_Limbs.M_CaretakerTransparency_Limbs")
          (Components|Activation|Activate _fx_idle false)
          (CallFunction|DebugCloseMaterialParamsLoop))))))

(event EventNPCAlertLevelChangedClient
  (params (NPC_AlertLevel :type Enum_NPC_AlertLevel%20Enum :default "NewEnumerator4"))
  ;[0.4219 1 0] Audio Handling
  ;;[1 1 1] Track Alert State For Audio (don't update on invalid or N/A)
  (switch Utilities|FlowControl|Switch|SwitchonEnum_NPC_AlertLevel (Variables|NPCAlertLevel|GetNPC_AlertLevel)
    (:NewEnumerator1
      (Variables|Sound|SetisCaretakerAggressiveAudio false)
      (CallFunction|UpdateCaretakerSoundLibrary))
    (:NewEnumerator2
      _)
    (:NewEnumerator3
      _))
  (Variables|Sound|SetisCaretakerAggressiveAudio true)
  (CallFunction|UpdateCaretakerSoundLibrary))

(event Custom|PlayVictorySound
  ;[0.4219 1 0] Audio Handling
  ;;[1 1 1] Play victory stringer after killing the caretake for players only in sideways.
  (if (Variables|Default|GetIsLocalPlayerInSideways)
    (Audio|PlaySound2D "/Caretaker/Sounds/Music/CareTaker_Victory_Stinger_Cue.CareTaker_Victory_Stinger_Cue")))

(event Custom|EnteredSidewaysClient
  ;[0.6123 0 1] Sideways Handling
  ;;[0 0.9498 1] This is for client side stuff, (VFX, materials, audio)
  (CallFunction|UpdateCaretakerSoundLibrary)
  (Components|Activation|Activate (Variables|Components|GetFX_Idle) false)
  (Effects|Components|ParticleSystem|SetFloatParameter (Variables|Components|GetFX_Idle) "Sideways" 1.0)
  (if (not (Variables|Default|GetisMaterialHandledFromPlayer))
    (CallFunction|SetCaretakerMaterials true) ; If player is not handling materials change, do it based on caretaker
  ))

(event Custom|ExitedSidewaysClient
  ;[0.6123 0 1] Sideways Handling
  ;;[0 0.9498 1] This is for client side stuff, (VFX, materials, audio)
  (CallFunction|UpdateCaretakerSoundLibrary)
  (Effects|Components|ParticleSystem|SetFloatParameter (Variables|Components|GetFX_Idle) "Sideways" 0.0)
  (if (not (Variables|Default|GetisMaterialHandledFromPlayer))
    (CallFunction|SetCaretakerMaterials false) ; If player is not handling materials change, do it based on caretaker
  )
  (Utilities|FlowControl|Delay 5.0)
  (Components|Activation|Deactivate (Variables|Components|GetFX_Idle)))

(event Custom|UpdateCaretakerSoundLibrary
  (bind (_getcurrentaggressivetagstate _bisvalidtagstate) (Sound|IsAggressiveorAlert))
  ;[0.4219 1 0] Audio Handling
  (if (Variables|Sound|GetisCaretakerAggressiveAudio)
    ;;[1 1 1] Agreessive Path
    (if (Variables|Default|GetisInSideways)
      (Sound|AddRemoveSoundLibrariestoSelf (Variables|Sound|GetSidewaysInsideLib) (Variables|Sound|GetSidewaysOutsideLib))
      (Sound|UpdateAlertLevelActorSate "NewEnumerator0")
      (else
        (Sound|AddRemoveSoundLibrariestoSelf (Variables|Sound|GetSidewaysOutsideLib) (Variables|Sound|GetSidewaysInsideLib))
        (Sound|UpdateTrackedActorState _getcurrentaggressivetagstate _bisvalidtagstate)))
    (else
      ;;[1 1 1] Unaware Path
      (if (Variables|Default|GetisInSideways)
        (Sound|AddRemoveSoundLibrariestoSelf (Variables|Sound|GetSidewaysInsideLib) (Variables|Sound|GetSidewaysOutsideLib))
        (Sound|UpdateAlertLevelActorSate "NewEnumerator0")
        (else
          (Sound|AddRemoveSoundLibrariestoSelf (Variables|Sound|GetSidewaysOutsideLib) (Variables|Sound|GetSidewaysInsideLib))
          (Sound|UpdateAlertLevelActorSate "NewEnumerator1"))))))

(event Custom|DebugSetMaterialParamsLoop
  (bind _mesh (Variables|Character|GetMesh))
  ;[1 0.0304 0] Debug Functionality For Caretaker
  ;;[1 1 1] Sets Materials Params
  (Utilities|FlowControl|Sequence
    (:then_0
      _)
    (:then_1
      _))
  (Utilities|FlowControl|Gate true
    (:Exit
      (Rendering|Material|SetScalarParameterValueonMaterials _mesh "SidewaysBubble_Scale_LocalP1" 2e+04)
      (Rendering|Material|SetVectorParameterValueonMaterials _mesh "SidewaysBubble_Location_LocalP1" (Transformation|GetActorLocation))
      (Utilities|FlowControl|DelayUntilNextTick))))

(event Custom|DebugCloseMaterialParamsLoop
  (bind _mesh (Variables|Character|GetMesh))
  ;[1 0.0304 0] Debug Functionality For Caretaker
  ;;[1 1 1] Sets Materials Params
  (Utilities|FlowControl|Gate true
    (:Exit
      (Rendering|Material|SetScalarParameterValueonMaterials _mesh "SidewaysBubble_Scale_LocalP1" 2e+04)
      (Rendering|Material|SetVectorParameterValueonMaterials _mesh "SidewaysBubble_Location_LocalP1" (Transformation|GetActorLocation))
      (Utilities|FlowControl|DelayUntilNextTick))))

(event EventOnDeathPlayEffects
  (params (Damage :type Float%20%28single-precision%29) (DamageTags :type Gameplay%20Tag%20Container%20Structure%20%28by%20ref%29) (Momentum :type Vector :default "0, 0, 0") (HitInfo :type Hit%20Result%20Structure%20%28by%20ref%29) (InstigatedBy :type Fort%20Pawn%20Object%20Reference) (DamageCauser :type Actor%20Object%20Reference) (EffectContext :type Gameplay%20Effect%20Context%20Handle%20Structure))
  (bind _mesh (Variables|Character|GetMesh))
  (bind _fx_idle (Variables|Components|GetFX_Idle))
  ;[0.6123 0 1] Sideways Handling
  (Rendering|Material|SetMaterial _mesh 0 "/Caretaker/Materials/M_Caretaker_Body_Death.M_Caretaker_Body_Death")
  (Rendering|Material|SetMaterial _mesh 1 "/Caretaker/Materials/M_Caretaker_Limbs_Death.M_Caretaker_Limbs_Death")
  ;[1 1 1] Death Material Override
  (Rendering|Material|SetMaterial _mesh 2 "/Game/Characters/Player/Male/Medium/Bodies/M_MED_HighTower_Tomato_Casual/Materials/M_VertexCrunch.M_VertexCrunch")
  (Rendering|Material|SetMaterial _mesh 3 "/Game/Characters/Player/Male/Medium/Bodies/M_MED_HighTower_Tomato_Casual/Materials/M_VertexCrunch.M_VertexCrunch")
  (Rendering|Material|SetScalarParameterValueonMaterials _mesh "GameTimeFromBP" (Utilities|Time|GetGameTimeinSeconds))
  (CallFunction|PlayVictorySound)
  (Effects|Components|ParticleSystem|SetFloatParameter _fx_idle "Sideways" 0.0)
  (Utilities|FlowControl|Delay 5.0)
  (Components|Activation|Deactivate _fx_idle))

(event Custom|DebugCloseMaterialParamsLoop_Multicast
  (bind _mesh (Variables|Character|GetMesh))
  ;[1 0.0304 0] Debug Functionality For Caretaker
  ;;[1 1 1] Sets Materials Params
  (Networking|SwitchHasAuthority
    (:Remote
      (Utilities|FlowControl|Gate true
        (:Exit
          (Rendering|Material|SetScalarParameterValueonMaterials _mesh "SidewaysBubble_Scale_LocalP1" 2e+04)
          (Rendering|Material|SetVectorParameterValueonMaterials _mesh "SidewaysBubble_Location_LocalP1" (Transformation|GetActorLocation))
          (Utilities|FlowControl|DelayUntilNextTick))))))

(event Custom|WatchForKeyEventsOnLocalController
  (params (possessingPlayer :type Fort%20Player%20Controller%20Object%20Reference))
  (bind _localpossessingplayer (Variables|Default|GetlocalPossessingPlayer))
  ;[1 0.0304 0] Debug Functionality For Caretaker
  ;;[1 1 1] Possession Handling For Capture
  (Variables|Default|SetlocalPossessingPlayer possessingPlayer)
  (Networking|SwitchHasAuthority
    (:Remote
      (Utilities|IsValid _localpossessingplayer
        (:Is%20Valid
          (Input|EnableInput _localpossessingplayer)
          (Utilities|FlowControl|Sequence
            (:then_0
              _)
            (:then_1
              _))
          (Utilities|FlowControl|Gate true
            (:Exit
              (bind _actionattempted (CallFunction|TestforKeyPressandExecuteAction))
              (if _actionattempted
                (Utilities|FlowControl|Delay 1.0) ; Best Debounce Ever
              )
              (Utilities|FlowControl|DelayUntilNextTick))))))))

(event Custom|WeaponCapsuleSetStartingValues
  (bind _weaponcapsulecollision (Variables|Default|GetWeaponCapsuleCollision))
  ;[1 0.8603 0] Weapon Capsule Handling (comes from Anim Notifies)
  (Variables|WeaponCapsule|SetWeaponCapsule_OriginalHalfHeight (Components|Capsule|GetScaledCapsuleHalfHeight _weaponcapsulecollision))
  (Variables|WeaponCapsule|SetWeaponCapsule_OriginalRelativeLocation (Variables|/Script/Engine.SceneComponent|GetRelativeLocation _weaponcapsulecollision))
  (Variables|WeaponCapsule|SetWeaponCapsule_OriginalCapsuleRadius (Components|Capsule|GetScaledCapsuleRadius _weaponcapsulecollision)))

(event EventOnDamagePlayEffects
  (params (Damage :type Float%20%28single-precision%29) (DamageTags :type Gameplay%20Tag%20Container%20Structure%20%28by%20ref%29) (Momentum :type Vector :default "0, 0, 0") (HitInfo :type Hit%20Result%20Structure%20%28by%20ref%29) (InstigatedBy :type Fort%20Pawn%20Object%20Reference) (DamageCauser :type Actor%20Object%20Reference) (EffectContext :type Gameplay%20Effect%20Context%20Handle%20Structure))
  (bind (_bblockinghit _binitialoverlap _time _distance _location _impactpoint _normal _impactnormal _physmat _hitactor _hitcomponent _hitbonename) (Collision|BreakHitResult HitInfo))
  ;[1 1 1] Overriding the hit react system: This uses hit location instead of hit direction (only if vulnerable)
  (if (Variables|FortPawn|GetbIsInvulnerable)
    ;[1 1 1] Play audio for invulnerable impacts. Just for local player.
    (bind _asfort_player_pawn (Utilities|Casting|CastToFortPlayerPawn InstigatedBy))
    (if (Pawn|IsLocallyViewed _asfort_player_pawn)
      (Audio|PlaySound2D "/Caretaker/Sounds/Caretaker_Damage_Impact_Invulnerable_Cue.Caretaker_Damage_Impact_Invulnerable_Cue"))
    (else
      (Utilities|FlowControl|Sequence
        (:then_0
          (if (and (not (Variables|Animation|GetbIsStaggered)) (not (Variables|Animation|GetbIsKnockedback)))
            (Utilities|FlowControl|DoOnce false
              (:Completed
                (Utilities|Time|SetTimerbyFunctionName 0 "AdditiveHitReactDelay" (Variables|NPCAdditiveHitReactions|GetAdditiveHitReactRetriggerDelay))
                (CallFunction|PlayAdditiveHitReactsCaretaker _hitbonename)))))
        (:then_1
          (switch Utilities|FlowControl|Switch|SwitchonEFortDamageZone (DamageZones|GetDamageZone HitInfo)
            (:Critical
              (GameplayCue|ExecuteGameplayCue%28Burst%29Local self (Variables|Default|GetGC_CriticalHit) (Ability|GameplayCue|MakeGameplayCueParameters :NormalizedMagnitude 0.0 :RawMagnitude 0.0 :Location _impactpoint :Normal "0,0,0" :GameplayEffectLevel 1 :AbilityLevel 1 :bReplicateLocationWhenUsingMinimalRepProxy false)))))))))

(event EventAdditiveHitReactDelay
  ;[1 1 1] Overriding the hit react system: This uses hit location instead of hit direction (only if vulnerable)
  (Utilities|FlowControl|DoOnce false
    (:Completed
      (Utilities|Time|SetTimerbyFunctionName 0 "AdditiveHitReactDelay" (Variables|NPCAdditiveHitReactions|GetAdditiveHitReactRetriggerDelay))
      (CallFunction|PlayAdditiveHitReactsCaretaker (.HitBoneName (Collision|BreakHitResult (stale EventOnDamagePlayEffects HitInfo)))))))

(event Custom|ServerRunAbilityByTag
  (params (GameplayTagContainer :type Gameplay%20Tag%20Container%20Structure%20%28by%20ref%29))
  ;[1 0.0304 0] Debug Functionality For Caretaker
  ;;[1 1 1] Possession Handling For Capture
  (Abilities|TryActivateAbilitiesbyTag (Variables|GameplayAbilities|GetAIPawnAbilitySystemComponent) GameplayTagContainer true))

(event Custom|RequestWeaponCapsuleChange
  (params (NewHalfHeight :type Float%20%28double-precision%29) (NewRelativeLocation :type Vector) (NewCapsuleRadius :type Float%20%28double-precision%29))
  (bind _weaponcapsulecollision (Variables|Default|GetWeaponCapsuleCollision))
  ;[1 0.8603 0] Weapon Capsule Handling (comes from Anim Notifies)
  (Math|Integer|IncrementInt (Variables|WeaponCapsule|GetWeaponCapsule_ChangeStackCount)
    (:%20%20
      (Transformation|SetRelativeLocation _weaponcapsulecollision (Math|Vector|MakeVector (.x NewRelativeLocation) (.y NewRelativeLocation) (+ (.z NewRelativeLocation) (- (.z (Variables|WeaponCapsule|GetWeaponCapsule_OriginalRelativeLocation)) (- (Variables|WeaponCapsule|GetWeaponCapsule_OriginalHalfHeight) NewHalfHeight)))) false true)
      (Components|Capsule|SetCapsuleSize _weaponcapsulecollision NewHalfHeight NewCapsuleRadius true))))

(event Custom|RequestStopWeaponCapsuleChange
  ;[1 0.8603 0] Weapon Capsule Handling (comes from Anim Notifies)
  (bind _result (Math|Integer|DecrementInt (Variables|WeaponCapsule|GetWeaponCapsule_ChangeStackCount))
    (:%20%20
      (if (<= _result 0)
        (CallFunction|FullResetWeaponCapsuleChange)))))

(event Custom|ResetBeamSweepTransform
  ;[1 1 1] Beam Sweep
  (CallFunction|BeamSweepResetMulticast)
  (Networking|ForceNetUpdate))

(event Custom|BeamSweepResetMulticast
  ;[1 1 1] Beam Sweep
  (Variables|BeamSweep|SetBeamSweep_BeamIdle true)
  (Variables|BeamSweep|SetBeamSweep_BeamEnd false))

(event Custom|BeginBeamSweep
  ;[1 1 1] Beam Sweep
  (CallFunction|BeamSweepUpdateOverTimeMulticast)
  (Networking|ForceNetUpdate))

(event Custom|FullResetWeaponCapsuleChange
  (bind _weaponcapsulecollision (Variables|Default|GetWeaponCapsuleCollision))
  ;[1 0.8603 0] Weapon Capsule Handling (comes from Anim Notifies)
  (Variables|WeaponCapsule|SetWeaponCapsule_ChangeStackCount 0)
  (Transformation|SetRelativeLocation _weaponcapsulecollision (Variables|WeaponCapsule|GetWeaponCapsule_OriginalRelativeLocation) false true)
  (Components|Capsule|SetCapsuleSize _weaponcapsulecollision (Variables|WeaponCapsule|GetWeaponCapsule_OriginalCapsuleRadius) (Variables|WeaponCapsule|GetWeaponCapsule_OriginalHalfHeight) true))

(event Custom|BeamSweepUpdateOverTimeMulticast
  ;[1 1 1] Beam Sweep
  (Variables|BeamSweep|SetBeamSweep_BeamIdle false)
  (bind (_direction _newtrack_0) (Timeline|BeamSweep PlayFromStart)
    (:Update
      (bind _transform (CallFunction|MakeTransformfromPitchLocation (Math|Float|Lerp (Variables|BeamSweep|GetBeamSweep_StartingPitch) (Variables|BeamSweep|GetBeamSweep_EndingPitch) _newtrack_0)))
      (Variables|BeamSweep|SetBeamSweep_AimingTransform _transform))
    (:Finished
      (Variables|BeamSweep|SetBeamSweep_BeamEnd true))))

(event Custom|FootstepDamageHandling
  (bind _footstepinplacetimerhandle (Variables|FootstepDamage|GetfootStepInPlaceTimerHandle))
  (bind _footstepcentertimerhandle (Variables|FootstepDamage|GetfootStepCenterTimerHandle))
  ;[0.02952 0 1] Footstep Handling (handling damage and faking the timing as best we can)
  (bind _asfort_ability_system_component (Utilities|Casting|CastToFortAbilitySystemComponent (Ability|GetAbilitySystemComponent self)))
  (Variables|Default|SetFortAbilitySystemComponent_Cached _asfort_ability_system_component)
  (Utilities|FlowControl|Sequence
    (:then_0
      _)
    (:then_1
      _))
  (Utilities|FlowControl|Gate true
    (:Exit
      (Utilities|FlowControl|Sequence
        (:then_0
          _)
        (:then_1
          (Utilities|FlowControl|Delay (ScalableFloat|GetValueatLevel (Variables|FootstepDamage|GetfootstepHandlerTickrateHF)))))
      (if (and (AI|Components|NavMovement|IsMovingonGround (Variables|Character|GetCharacterMovement)) (> (Math|Vector|VectorLength (Variables|/Script/Engine.MovementComponent|GetVelocity (Variables|Character|GetCharacterMovement))) 1e+02))
        (if (not (Utilities|Time|IsValidTimerHandle (Variables|FootstepDamage|GetfootStepCenterTimerHandle)))
          (CallFunction|CaretakerTestandClearandInvalidateTimerbyHandle (Variables|FootstepDamage|GetfootStepFinalStepTimerHandle))
          (bind _returnvalue (Utilities|Time|SetTimerbyEvent (EventDispatchers|CreateEvent :!event Custom|ActivateFootstepDamageCenter) (ScalableFloat|GetValueatLevel (Variables|FootstepDamage|GetfootstepWalkStepIntervalHF)) true))
          (Variables|FootstepDamage|SetfootStepCenterTimerHandle _returnvalue))
        (elif (Abilities|HasActiveAbilitywithAnyMatchingTag (Variables|Default|GetFortAbilitySystemComponent_Cached) "(GameplayTags=((TagName=\"NPC.Ability.NonAttack.TurnTransition\")))")
          (CallFunction|CaretakerTestandClearandInvalidateTimerbyHandle (Variables|FootstepDamage|GetfootStepCenterTimerHandle))
          (if (not (Utilities|Time|IsValidTimerHandle (Variables|FootstepDamage|GetfootStepInPlaceTimerHandle)))
            (CallFunction|CaretakerTestandClearandInvalidateTimerbyHandle (Variables|FootstepDamage|GetfootStepFinalStepTimerHandle))
            (bind _returnvalue_2 (Utilities|Time|SetTimerbyEvent (EventDispatchers|CreateEvent :!event Custom|ActivateFootstepDamageCenter) (ScalableFloat|GetValueatLevel (Variables|FootstepDamage|GetfootstepTurnInPlaceStepIntervalHF)) true))
            (Variables|FootstepDamage|SetfootStepInPlaceTimerHandle _returnvalue_2))
          (elif (or (Utilities|Time|IsValidTimerHandle _footstepcentertimerhandle) (Utilities|Time|IsValidTimerHandle _footstepinplacetimerhandle))
            (Utilities|Time|ClearandInvalidateTimerbyHandle _footstepcentertimerhandle)
            (Utilities|Time|ClearandInvalidateTimerbyHandle _footstepinplacetimerhandle)
            (bind _returnvalue_3 (Utilities|Time|SetTimerbyEvent (EventDispatchers|CreateEvent :!event Custom|ActivateFootstepDamageInPlace) (ScalableFloat|GetValueatLevel (Variables|FootstepDamage|GetfootstepFinalStepDelayHF)))) ; Play one final step (feet settling)
            (Variables|FootstepDamage|SetfootStepFinalStepTimerHandle _returnvalue_3)))))))

(event Custom|StopFootStepDamageHandling
  (bind _footstepinplacetimerhandle (Variables|FootstepDamage|GetfootStepInPlaceTimerHandle))
  (bind _footstepcentertimerhandle (Variables|FootstepDamage|GetfootStepCenterTimerHandle))
  ;[0.02952 0 1] Footstep Handling (handling damage and faking the timing as best we can)
  (Utilities|Time|ClearandInvalidateTimerbyHandle (Variables|FootstepDamage|GetfootStepCenterTimerHandle))
  (Utilities|Time|ClearandInvalidateTimerbyHandle (Variables|FootstepDamage|GetfootStepInPlaceTimerHandle))
  (Utilities|Time|ClearandInvalidateTimerbyHandle (Variables|FootstepDamage|GetfootStepFinalStepTimerHandle))
  (Utilities|FlowControl|Gate true
    (:Exit
      (Utilities|FlowControl|Sequence
        (:then_0
          _)
        (:then_1
          (Utilities|FlowControl|Delay (ScalableFloat|GetValueatLevel (Variables|FootstepDamage|GetfootstepHandlerTickrateHF)))))
      (if (and (AI|Components|NavMovement|IsMovingonGround (Variables|Character|GetCharacterMovement)) (> (Math|Vector|VectorLength (Variables|/Script/Engine.MovementComponent|GetVelocity (Variables|Character|GetCharacterMovement))) 1e+02))
        (if (not (Utilities|Time|IsValidTimerHandle (Variables|FootstepDamage|GetfootStepCenterTimerHandle)))
          (CallFunction|CaretakerTestandClearandInvalidateTimerbyHandle (Variables|FootstepDamage|GetfootStepFinalStepTimerHandle))
          (bind _returnvalue (Utilities|Time|SetTimerbyEvent (EventDispatchers|CreateEvent :!event Custom|ActivateFootstepDamageCenter) (ScalableFloat|GetValueatLevel (Variables|FootstepDamage|GetfootstepWalkStepIntervalHF)) true))
          (Variables|FootstepDamage|SetfootStepCenterTimerHandle _returnvalue))
        (elif (Abilities|HasActiveAbilitywithAnyMatchingTag (Variables|Default|GetFortAbilitySystemComponent_Cached) "(GameplayTags=((TagName=\"NPC.Ability.NonAttack.TurnTransition\")))")
          (CallFunction|CaretakerTestandClearandInvalidateTimerbyHandle (Variables|FootstepDamage|GetfootStepCenterTimerHandle))
          (if (not (Utilities|Time|IsValidTimerHandle (Variables|FootstepDamage|GetfootStepInPlaceTimerHandle)))
            (CallFunction|CaretakerTestandClearandInvalidateTimerbyHandle (Variables|FootstepDamage|GetfootStepFinalStepTimerHandle))
            (bind _returnvalue_2 (Utilities|Time|SetTimerbyEvent (EventDispatchers|CreateEvent :!event Custom|ActivateFootstepDamageCenter) (ScalableFloat|GetValueatLevel (Variables|FootstepDamage|GetfootstepTurnInPlaceStepIntervalHF)) true))
            (Variables|FootstepDamage|SetfootStepInPlaceTimerHandle _returnvalue_2))
          (elif (or (Utilities|Time|IsValidTimerHandle _footstepcentertimerhandle) (Utilities|Time|IsValidTimerHandle _footstepinplacetimerhandle))
            (Utilities|Time|ClearandInvalidateTimerbyHandle _footstepcentertimerhandle)
            (Utilities|Time|ClearandInvalidateTimerbyHandle _footstepinplacetimerhandle)
            (bind _returnvalue_3 (Utilities|Time|SetTimerbyEvent (EventDispatchers|CreateEvent :!event Custom|ActivateFootstepDamageInPlace) (ScalableFloat|GetValueatLevel (Variables|FootstepDamage|GetfootstepFinalStepDelayHF)))) ; Play one final step (feet settling)
            (Variables|FootstepDamage|SetfootStepFinalStepTimerHandle _returnvalue_3)))))))

(event Custom|CastingSidewaysStatus
  (params (isCasting :type Boolean))
  ;[1 1 1] Tracking Sideways Encounter
  ;;[1 1 1] Casting Sideways
  (Variables|SafeZoneTracking|SetactivelyCastingSideways isCasting))

(event Custom|ActivateFootstepDamageCenter
  ;[0.02952 0 1] Footstep Handling (handling damage and faking the timing as best we can)
  (Ability|SendGameplayEventtoActor self (Variables|FootstepDamage|GetT_FootstepCenter) (Utilities|Struct|MakeGameplayEventData (Variables|FootstepDamage|GetT_FootstepCenter))) ; For moving footsteps
)

(event Custom|SidewaysPlayspaceMatchEnded
  ;[1 1 1] Tracking Sideways Encounter
  ;;[1 1 1] On Sideways Encounter Ended
  (Playspace|UnbindEventfromNotifyEndMatch (Variables|Default|GetSideways%20Playspace) (delegate Custom|SidewaysPlayspaceMatchEnded))
  (Variables|Default|SetSideways%20Playspace 0)
  (GameplayEffects|FortApplyGameplayEffectToSelf :self (Variables|GameplayAbilities|GetAIPawnAbilitySystemComponent) :GameplayEffectClass "/Caretaker/Abilities/GE_NPC_Irwin_Monster_Caretaker_CastSidewaysEnded_CooldownOverride.GE_NPC_Irwin_Monster_Caretaker_CastSidewaysEnded_CooldownOverride_C" :Level 0.0))

(event Custom|ActivateFootstepDamageInPlace
  ;[0.02952 0 1] Footstep Handling (handling damage and faking the timing as best we can)
  (Ability|SendGameplayEventtoActor self (Variables|FootstepDamage|GetT_FootstepInPlace) (Utilities|Struct|MakeGameplayEventData (Variables|FootstepDamage|GetT_FootstepInPlace))) ; For turning in place and stopping
)

(event Custom|MeleeWarningAttackTrigger
  ;[1 1 1] Melee Warning Attack Call for GameplayCues
  (CallFunction|MeleeWarningAttackTriggerMulticast)
  (Networking|ForceNetUpdate))

(event Custom|MeleeWarningAttackTriggerMulticast
  ;[1 1 1] Melee Warning Attack Call for GameplayCues
  (Networking|SwitchHasAuthority
    (:Remote
      (Default|CallMeleeWarningAttackFired))))

(event Custom|BeginCaretakerSafeZoneTracking
  ;[1 1 1] Caretaker Safe Zone Handling (escape into the storm!)
  (bind _returnvalue (Utilities|Time|SetTimerbyEvent (delegate Custom|CaretakerSafeZoneTrackingHandler) (ScalableFloat|GetValueatLevel (Variables|SafeZoneTracking|GetCaretakerSafeZonePhaseEscapeThreshold)) true))
  (Variables|SafeZoneTracking|SetCaretakerSafeZoneTrackingTimerHandle _returnvalue))

(event Custom|CaretakerSafeZoneTrackingHandler
  (bind _returnvalue (SafeZone|GetSafeZoneIndicator))
  ;[1 1 1] Caretaker Safe Zone Handling (escape into the storm!)
  (Utilities|IsValid _returnvalue
    (:Is%20Valid
      (if (Math|Boolean|ANDBoolean (>= (SafeZone|GetCurrentPhaseIndex _returnvalue) (Math|Float|Truncate (ScalableFloat|GetValueatLevel (Variables|SafeZoneTracking|GetCaretakerSafeZonePhaseEscapeThreshold)))) (not (Variables|Default|GetisInSideways)) (not (Variables|SafeZoneTracking|GetactivelyCastingSideways)))
        (Utilities|Time|ClearandInvalidateTimerbyHandle (Variables|SafeZoneTracking|GetCaretakerSafeZoneTrackingTimerHandle))
        (AI|Components|Blackboard|SetValueasVector (AI|GetBlackboard 0) (Variables|SafeZoneTracking|GetBBKey_SpawnLocation) (SafeZone|GetSafeZoneCenter _returnvalue))
        (GameplayEffects|ApplyGameplayEffectToSelf :self (Variables|GameplayAbilities|GetAIPawnAbilitySystemComponent) :GameplayEffectClass "/Caretaker/Abilities/GE_NPC_Irwin_Monster_Caretaker_EscapeIntoStorm.GE_NPC_Irwin_Monster_Caretaker_EscapeIntoStorm_C" :Level 0.0)))))

(event WaterBody|EventOnEnteredWaterBody
  (params (WaterBody :type Fort%20Water%20Body%20Actor%20Object%20Reference) (bIsFirstBody :type Boolean))
  ;[1 1 1] Track if Caretaker is in water
  (Variables|Default|SetIsInWater true))

(event WaterBody|EventOnExitedWaterBody
  (params (WaterBody :type Fort%20Water%20Body%20Actor%20Object%20Reference) (bIsLastBody :type Boolean))
  ;[1 1 1] Track if Caretaker is in water
  (Variables|Default|SetIsInWater false))

(fn AddRemoveSoundLibrariesToSelf
  (params (AddLibraries :type Array%20of%20Sound%20Library%20Object%20References) (RemoveLibraries :type Array%20of%20Sound%20Library%20Object%20References))
  (bind _returnvalue (WorldSubsystems|GetSoundLibrarySubsystem))
  (bind _self self)
  (Utilities|IsValid _returnvalue
    (:Is%20Valid
      (SoundLibrary|AddLibraries _returnvalue _self AddLibraries :!class /Script/SoundLibrary.SoundLibrarySubsystem)
      (SoundLibrary|RemoveLibraries _returnvalue _self RemoveLibraries :!class /Script/SoundLibrary.SoundLibrarySubsystem))))

(fn BlueprintModifyIncomingDamage
  (params (Damage :type Float%20%28single-precision%29) (DamageTags :type Gameplay%20Tag%20Container%20Structure%20%28by%20ref%29) (EffectContext :type Gameplay%20Effect%20Context%20Handle%20Structure%20%28by%20ref%29) (EventInstigator :type Controller%20Object%20Reference) (DamageCauser :type Actor%20Object%20Reference))
  (bind _returnvalue (Pawn|GetControlledPawn EventInstigator))
  ;[1 1 1] If the instigator is not in the sideways no damage is granted
  (bind _returnvalue_2 (|CallParentFunction Damage DamageTags EffectContext EventInstigator DamageCauser))
  (if (Variables|Damage|GetmodifyDamageWithSidewaysCheck)
    (Utilities|IsValid _returnvalue
      (:Is%20Valid
        (Utilities|IsValid _returnvalue
          (:Is%20Valid
            (bind _returnvalue_3 (GameplayTags|HasMatchingGameplayTag(Message) _returnvalue "(TagName=\"Granted.Athena.Sideways.InsideSideways\")"))
            (if _returnvalue_3
              (return _returnvalue_2)
              (else
                _)))
          (:Is%20Not%20Valid
            _))
        (return 0.0))
      (:Is%20Not%20Valid
        _))
    (else
      (return _returnvalue_2))))

(fn Caretaker_BeamSweep_Transform
  (bind _beamsweep_beamidle (Variables|BeamSweep|GetBeamSweep_BeamIdle))
  (if (or _beamsweep_beamidle (Variables|BeamSweep|GetBeamSweep_BeamEnd))
    (bind _transform (CallFunction|MakeTransformfromPitchLocation (Math|Float|SelectFloat (Variables|BeamSweep|GetBeamSweep_StartingPitch) (Variables|BeamSweep|GetBeamSweep_EndingPitch) _beamsweep_beamidle)))
    (bind _output_get (Variables|BeamSweep|SetBeamSweep_AimingTransform _transform))
    (return _output_get)
    (else
      (return (Variables|BeamSweep|GetBeamSweep_AimingTransform)))))

(fn CaretakerTestAndClearAndInvalidateTimerByHandle
  (params (Handle :type Timer%20Handle%20Structure%20%28by%20ref%29))
  (bind _output_get (LocalVariables|SetisTimerValid (Utilities|Time|IsValidTimerHandle Handle)))
  (if _output_get
    (Utilities|Time|ClearandInvalidateTimerbyHandle Handle)
    (return (LocalVariables|GetisTimerValid))))

(fn GetMontageBasedOnClosestBone
  (params (HitBoneName :type Name))
  (bind _hittestbonename (LocalVariables|GetHitTestBoneName))
  (LocalVariables|SetHitTestBoneName (Utilities|String|ToString(Name) HitBoneName))
  (if (or (Utilities|String|Contains _hittestbonename "L_BipedArm") (Utilities|String|Contains _hittestbonename "L_BipedHand"))
    (return "/Caretaker/Animations/Caretaker_HitReact_LeftArm_M.Caretaker_HitReact_LeftArm_M")
    (elif (or (Utilities|String|Contains _hittestbonename "R_BipedArm") (Utilities|String|Contains _hittestbonename "R_BipedHand"))
      (return "/Caretaker/Animations/Caretaker_HitReact_RightArm_M.Caretaker_HitReact_RightArm_M")
      (else
        (return "/Caretaker/Animations/Caretaker_HitReact_Body_M.Caretaker_HitReact_Body_M")))))

(fn IsAggressiveOrAlert
  (bind _npc_alertleveltotagmap (Variables|NPCAlertLevel|GetNPC%20AlertLevelToTagMap))
  (bind (_value _returnvalue) (Utilities|Map|Find _npc_alertleveltotagmap "NewEnumerator2"))
  (bind _npc_alertleveltotagmap_2 (Variables|NPCAlertLevel|GetNPC%20AlertLevelToTagMap))
  (bind (_value_2 _returnvalue_2) (Utilities|Map|Find _npc_alertleveltotagmap_2 "NewEnumerator3"))
  (bind _returnvalue_3 (or _returnvalue_2 _returnvalue))
  (if _returnvalue_3
    (return (select _returnvalue _value _value_2) _returnvalue_3)))

(fn IsValidAutoFireTarget
  (params (Attacker :type Actor%20Object%20Reference))
  (return (Variables|Default|GetisInSideways)))

(fn MakeTransformFromPitchLocation
  (params (Pitch :type Float%20%28double-precision%29))
  (Networking|SwitchHasAuthority
    (:Authority
      ;[1 1 1] On Server Use a local offset from root as an approx position
      (LocalVariables|SettargetLocation (- (Math|Transform|TransformLocation (Transformation|GetActorTransform) (Variables|BeamSweep|GetBeamSweep_ServerApproxBeamOrigin)) (Math|Vector|MakeVector 0.0 0.0 (Components|Capsule|GetScaledCapsuleHalfHeight (Variables|Character|GetCapsuleComponent))))))
    (:Remote
      ;[1 1 1] On Client Use the FX_Chest Location
      (LocalVariables|SettargetLocation (.location (Transformation|GetSocketTransform (Variables|Character|GetMesh) (Variables|BeamSweep|GetBeamSweep_SourceSocket) "RTS_World")))))
  (return (Math|Transform|MakeTransform (LocalVariables|GettargetLocation) (Math|Rotator|MakeRotator 0.0 (Variables|GetPitch) (.yaw (.rotation (Transformation|GetActorTransform)))))))

(fn OnRep_isCaretakerHidden
  (if (not (Variables|Default|GetisCaretakerHidden))
    (Rendering|SetActorHiddenInGame false)))

(fn OnRep_isInSideways
  (Networking|SwitchHasAuthority
    (:Remote
      (if (Variables|Default|GetisInSideways)
        (CallFunction|EnteredSidewaysClient)
        (else
          (CallFunction|ExitedSidewaysClient))))))

(fn OnRep_IsInWater
  (bind _returnvalue (Class|FortAnimInputEventBool|CreateAnimInputEventBool "aie.CaretakerInWater" (Variables|Default|GetIsInWater)))
  (Animation|TriggerAnimInputEvent _returnvalue))

(fn PlayAdditiveHitReactsCaretaker
  (params (HitBoneName :type Name))
  (bind _selectedmontage (CallFunction|GetMontageBasedonClosestBone HitBoneName))
  (Animation|PlayLocalAnimMontage _selectedmontage))

(fn TestForKeyPressAndExecuteAction
  (if (Game|Player|WasInputKeyJustPressed (Variables|Default|GetlocalPossessingPlayer) "F")
    (CallFunction|ServerRunAbilitybyTag (LocalVariables|GetGameplayTagLocal1))
    (elif (Game|Player|WasInputKeyJustPressed (Variables|Default|GetlocalPossessingPlayer) "G")
      (CallFunction|ServerRunAbilitybyTag (LocalVariables|GetGameplayTagLocal2))
      (elif (Game|Player|WasInputKeyJustPressed (Variables|Default|GetlocalPossessingPlayer) "T")
        (CallFunction|ServerRunAbilitybyTag (LocalVariables|GetGameplayTagLocal3))
        (elif (Game|Player|WasInputKeyJustPressed (Variables|Default|GetlocalPossessingPlayer) "V")
          (CallFunction|ServerRunAbilitybyTag (LocalVariables|GetGameplayTagLocal4))
          (elif (Game|Player|WasInputKeyJustPressed (Variables|Default|GetlocalPossessingPlayer) "B")
            (CallFunction|ServerRunAbilitybyTag (LocalVariables|GetGameplayTagLocal5))
            (elif (Game|Player|WasInputKeyJustPressed (Variables|Default|GetlocalPossessingPlayer) "N")
              (CallFunction|ServerRunAbilitybyTag (LocalVariables|GetGameplayTagLocal5))
              (elif (Game|Player|WasInputKeyJustPressed (Variables|Default|GetlocalPossessingPlayer) "R")
                (CallFunction|DebugCaretakerToggleSideways)
                (CallFunction|DebugCaretakerToggleMaterial)
                (else
                  (return false)))))))))
  (Networking|ForceNetUpdate)
  (return true))

(fn UpdateAlertLevelActorSate
  (params (AlertState :type Enum_NPC_AlertLevel%20Enum :default "NewEnumerator4"))
  (bind (_value _returnvalue) (Utilities|Map|Find (Variables|NPCAlertLevel|GetNPC%20AlertLevelToTagMap) AlertState))
  ;[1 1 1] Set the alert level tag
  (Sound|UpdateTrackedActorState _value _returnvalue))

(fn UpdateTrackedActorState
  (params (CurrentState :type Gameplay%20Tag%20Structure%20%28by%20ref%29) (bIsValidTag :type Boolean))
  ;[1 1 1] Set the data tracker actor state so we can play proper music.
  (bind _returnvalue (FortGameplayDataTrackerComponent|GetFortGameplayDataTrackerComponentManager))
  (Utilities|IsValid _returnvalue
    (:Is%20Valid
      (if bIsValidTag
        (Class|FortGameplayDataTrackerComponentManager|SetTrackedActorState _returnvalue self CurrentState)))))

(fn ConstructionScript
  (|CallParentFunction))
