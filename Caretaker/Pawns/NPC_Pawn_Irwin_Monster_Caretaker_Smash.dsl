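; Caretaker monster NPC pawn variant for the Smash/Mash mode that starts out visible and vulnerable and drops perk loot on death.
; BeginPlay calls the EnteredSideways functions directly on server and client, bypassing the state change that requires a sideways playspace.
; When enabled, it registers itself as a sound indicator for each player in the Mash mutator's data.
; On death, the server rolls LTG_Perks once and tosses an instanced pickup per player.
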(event EventBeginPlay
  ;[0.05517 0.3316 0.1095] We cannot set the caretaker directly into "in sideways" state as the sideways component has a validity check for needing a sideways playspace to be in "in sideways" state. Instead, call the functions that will make the Caretaker visible and vulnerable directly.
  (|CallParentFunction)
  (Utilities|FlowControl|Sequence
    (:then_0
      (Default|MashMarkforPlayers)
      (Networking|SwitchHasAuthority
        (:Authority
          (CallFunction|EnteredSideways)
          (EventDispatchers|BindEventtoOnDied (delegate Custom|Died_SpawnLoot)))
        (:Remote
          (CallFunction|EnteredSidewaysClient)
          (CallFunction|SetCaretakerMaterials true))))
    (:then_1
      ; Pulse Lancer Parameters
      (bind (_gamestate _playlist _playlistcontexttags) (Class|FortAsyncActionCurrentPlaylistReady|PlaylistDataReadyAsync)
        (:OnReady
          (if (GameplayTags|HasTag _playlistcontexttags "(TagName=\"Athena.Playlist.Mash\")" true)
            _))))))

(event Custom|Died_SpawnLoot
  (params (DamagedActor :type Actor%20Object%20Reference) (Damage :type Float%20%28single-precision%29) (InstigatedBy :type Controller%20Object%20Reference) (DamageCauser :type Actor%20Object%20Reference) (HitLocation :type Vector) (FHitComponent :type Primitive%20Component%20Object%20Reference) (BoneName :type Name) (Momentum :type Vector))
  (bind _gamestate (Class|FortAsyncActionCurrentPlaylistReady|PlaylistDataReadyAsync)
    (:OnReady
      (bind _outloottodrop (Loot|PickLootDrops "LTG_Perks" 0 -1 "(GameplayTags=)"))
      (for _ (range (Variables|/Script/FortniteGame.FortGameStateAthena|GetTotalPlayers _gamestate))
        (Inventory|SpawnPickupInWorldWithClassAndItemEntryForInstancedLoot (Utilities|Array|Get(aref) _outloottodrop) 0 (Transformation|GetActorLocation) "0.000000,0.000000,1.000000" 0 true true false "Tossed" "Unset" 0 false)))))

(event EventOnSpectatingTargetChangedEvent_Event_0
  (params (SpectatingTarget :type Fort%20Player%20State%20Zone%20Object%20Reference)))

(event EventBeginCaretakerSafeZoneTracking)

(event EventCaretakerSafeZoneTrackingHandler)

(fn Mash_MarkForPlayers
  (if (Variables|Default|GetbMash_ShouldMarkForPlayers)
    (bind _asfort_game_state_athena (Utilities|Casting|CastToFortGameStateAthena (Game|GetGameState)))
    (bind _asmash_mutator (Utilities|Casting|CastToMashMutator (Modifier|GetMutatorbyClass (Variables|/Script/FortniteGame.FortGameStateAthena|GetMutatorListComponent _asfort_game_state_athena) "/Game/Athena/Playlists/Mash/MashMutator.MashMutator_C")))
    (bind _playerdataarray_entries (Utilities|Struct|BreakMashPlayerDataArray (Variables|/Script/FortniteGame.FortAthenaMutator_Mash|GetPlayerDataArray_Entries _asmash_mutator)))
    (for _array_element _playerdataarray_entries
      (Fortnite|RegisterSoundIndicatorPlacementContext (Fortnite|GetFortPlayerControllerfromActor (Utilities|Struct|BreakMashPlayerDataEntry _array_element)) self))))

(fn ConstructionScript
  (|CallParentFunction))
